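; Spawn action for always-spawn Valet BasicCar vehicles that marks them inoperable after all actors spawn.
; It loops over the spawned actors, casts each to the Fort vehicle interface and calls MakeInoperable.
; The call only happens when the all-Valet-inoperable scalable float is true or the BasicCar inoperable chance equals exactly one, so partial chances are not rolled here.
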(fn OnAllActorSpawned
  (params (SpawnedActor :type Array%20of%20Actor%20Object%20References) (SourcePointProvider :type Fort%20Athena%20Living%20World%20Point%20Provider%20Interface%20Interface%20%28by%20ref%29))
  ;[1 1 1] Separate ability to handle marking Always Spawn vehicles as inoperable
  (for _array_element SpawnedActor
    (bind _asfort_vehicle_interface (Utilities|Casting|CastToFortVehicleInterface _array_element))
    (if (or (ScalableFloat|GetValueasBool (Variables|Default|GetAllValetInoperable)) (== (ScalableFloat|GetValueatLevel (Variables|Default|GetBasicCarInoperableChance)) 1.0))
      (Vehicle|MakeInoperable _asfort_vehicle_interface)))
  (return))
